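; Looping gameplay cue notify for the Black Monday cape glider deploy effect.
; On loop start, the FX-system version of the event casts the target to the Black Monday legacy glider Blueprint and reads its IsGrapple flag.
; When the flag is set, it deactivates the first spawned particle component, suppressing that effect for grapple use.
; The Cascade-only event and construction script are empty.
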(event EventOnLoopingStart(Cascade)
  (params (MyTarget :type Actor%20Object%20Reference) (Parameters :type Gameplay%20Cue%20Parameters%20Structure%20%28by%20ref%29) (ParticleComponents :type Array%20of%20Cascade%20Particle%20System%20Component%20Object%20References) (AudioComponents :type Array%20of%20Audio%20Component%20Object%20References)))

(event EventOnLoopingStart(EitherVFX-preferthisversion)
  (params (MyTarget :type Actor%20Object%20Reference) (Parameters :type Gameplay%20Cue%20Parameters%20Structure%20%28by%20ref%29) (ParticleComponents :type Array%20of%20FXSystem%20Component%20Object%20References) (AudioComponents :type Array%20of%20Audio%20Component%20Object%20References))
  (bind _asb_athena_glider_legacy_black_monday_cake (Utilities|Casting|CastToB_Athena_Glider_Legacy_BlackMondayCake MyTarget))
  (if (Variables|/Game/Athena/Cosmetics/Blueprints/Gliders/LegacyGliderBPs/B_Athena_Glider_Legacy_BlackMondayCake.B_Athena_Glider_Legacy_BlackMondayCake_C|GetIsGrapple _asb_athena_glider_legacy_black_monday_cake)
    (Components|Activation|Deactivate (Utilities|Array|Get(acopy) ParticleComponents))))

(fn ConstructionScript)
